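# Post-copy secondary-motion AnimBP for the female Commando Medic body, animating stethoscope, kneepad and front pouch.
# It copies the attached parent's pose, then chains AnimDynamics, bone-driven controllers and a pouch Look At.
# In the front end, spine_05 yaw relative to spine_03 sets the alpha of additive stethoscope corrections, otherwise zero.
# An RBF Pose Driver for the pouch and an AimOffset calculation are present but unconnected.
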AnimBlueprint /Game/Characters/Player/Female/Medium/Bodies/F_MED_Commando_Medic/Meshes/F_MED_Medic_Export_AnimBP
target_skeleton=&/F_MED_Medic_Export_Skeleton
graphs [
 AnimGraph {nodes [
   Pose Driver - Source: spine_01!(SourcePose:Pose Link Structure) -> (Pose:Pose Link Structure) {
    is=title AnimGraphNode_PoseDriver
    poseAsset=&/F_MED_Medic_RBF_PoseAsset
    onlyDriveBones [{boneName=dyn_front_pouch}]
    poseTargets [
     {
      bApplyCustomCurve=false
      bIsHidden=false
      distanceMethod=DefaultMethod
      drivenName=Pose_1
      functionType=DefaultFunction
      targetScale=0.819
      boneTransforms [
       {targetRotation=<rot 0 -3.9009 0>, targetTranslation=<6.1208 -0.9328 0>}
       {targetRotation=<rot 0 -2.7227 0>, targetTranslation=<6.5996 -0.2095 0>}
       {targetRotation=<rot 0 -1.7047 0>, targetTranslation=<6.4914 -0.1237 0>}
       {targetRotation=<rot 0 1.5078 0>, targetTranslation=<6.9598 0.0794 0>}
       {targetRotation=<rot 0 0.1712 0>, targetTranslation=<6.9987 0 0>}
      ]
      customCurve {defaultValue=3.4028234663852886e+38, postInfinityExtrap=RCCE_Constant, preInfinityExtrap=RCCE_Constant, keys []}
      targetRotation=<rot 0 0 0>
     }
     {
      bApplyCustomCurve=false
      bIsHidden=false
      distanceMethod=DefaultMethod
      drivenName=Pose_2
      functionType=DefaultFunction
      targetScale=0.7275
      boneTransforms [
       {targetRotation=<rot 1.7004 9.0187 -1.3081>, targetTranslation=<6.1208 -0.9328 0>}
       {targetRotation=<rot -0.2437 -3.349 -2.4594>, targetTranslation=<6.5996 -0.2095 0>}
       {targetRotation=<rot -0.238 -5.6122 -2.6799>, targetTranslation=<6.4914 -0.1237 0>}
       {targetRotation=<rot -0.2438 -8.7166 -2.5393>, targetTranslation=<6.9598 0.0794 0>}
       {targetRotation=<rot 0 0.1712 0>, targetTranslation=<6.9987 0 0>}
      ]
      customCurve {defaultValue=3.4028234663852886e+38, postInfinityExtrap=RCCE_Constant, preInfinityExtrap=RCCE_Constant, keys []}
      targetRotation=<rot 0 0 0>
     }
     {
      bApplyCustomCurve=false
      bIsHidden=false
      distanceMethod=DefaultMethod
      drivenName=Pose_3
      functionType=DefaultFunction
      targetScale=0.7275
      boneTransforms [
       {targetRotation=<rot 3.3642 9.0532 -4.7316>, targetTranslation=<6.1208 -0.9328 0>}
       {targetRotation=<rot -0.595 3.4437 -4.1017>, targetTranslation=<6.5996 -0.2095 0>}
       {targetRotation=<rot -2.0233 0.6011 -3.5986>, targetTranslation=<6.4914 -0.1237 0>}
       {targetRotation=<rot -3.217 -2.7561 -3.0528>, targetTranslation=<6.9598 0.0794 0>}
       {targetRotation=<rot 0 0.1712 0>, targetTranslation=<6.9987 0 0>}
      ]
      customCurve {defaultValue=3.4028234663852886e+38, postInfinityExtrap=RCCE_Constant, preInfinityExtrap=RCCE_Constant, keys []}
      targetRotation=<rot 0 0 0>
     }
     {
      bApplyCustomCurve=false
      bIsHidden=false
      distanceMethod=DefaultMethod
      drivenName=Pose_4
      functionType=DefaultFunction
      targetScale=1
      boneTransforms [
       {targetRotation=<rot 3.1297 -3.2244 -6.6312>, targetTranslation=<6.1208 -0.9328 0>}
       {targetRotation=<rot -0.2303 14.4554 -4.2472>, targetTranslation=<6.5996 -0.2095 0>}
       {targetRotation=<rot -2.7692 13.8853 -2.7056>, targetTranslation=<6.4914 -0.1237 0>}
       {targetRotation=<rot -4.9264 16.3162 -1.6233>, targetTranslation=<6.9598 0.0794 0>}
       {targetRotation=<rot 0 0.1712 0>, targetTranslation=<6.9987 0 0>}
      ]
      customCurve {defaultValue=3.4028234663852886e+38, postInfinityExtrap=RCCE_Constant, preInfinityExtrap=RCCE_Constant, keys []}
      targetRotation=<rot 0 0 0>
     }
     {
      bApplyCustomCurve=false
      bIsHidden=false
      distanceMethod=DefaultMethod
      drivenName=Pose_5
      functionType=DefaultFunction
      targetScale=0.8096
      boneTransforms [
       {targetRotation=<rot 0.1605 1.7596 -4.2956>, targetTranslation=<6.1208 -0.9328 0>}
       {targetRotation=<rot 3.8879 -3.8028 -4.7068>, targetTranslation=<6.5996 -0.2095 0>}
       {targetRotation=<rot 4.3817 -3.8729 -4.8683>, targetTranslation=<6.4914 -0.1237 0>}
       {targetRotation=<rot 4.2745 -1.99 -4.8946>, targetTranslation=<6.9598 0.0794 0>}
       {targetRotation=<rot 0 0.1712 0>, targetTranslation=<6.9987 0 0>}
      ]
      customCurve {defaultValue=3.4028234663852886e+38, postInfinityExtrap=RCCE_Constant, preInfinityExtrap=RCCE_Constant, keys []}
      targetRotation=<rot 0 0 0>
     }
     {
      bApplyCustomCurve=false
      bIsHidden=false
      distanceMethod=DefaultMethod
      drivenName=Pose_6
      functionType=DefaultFunction
      targetScale=1
      boneTransforms [
       {targetRotation=<rot 1.3642 -4.2847 -7.1517>, targetTranslation=<6.1208 -0.9328 0>}
       {targetRotation=<rot -0.2459 3.7117 -6.7391>, targetTranslation=<6.5996 -0.2095 0>}
       {targetRotation=<rot -2.0552 5.4815 -6.9845>, targetTranslation=<6.4914 -0.1237 0>}
       {targetRotation=<rot -6.2131 10.5856 -7.5148>, targetTranslation=<6.9598 0.0794 0>}
       {targetRotation=<rot 0 0.1712 0>, targetTranslation=<6.9987 0 0>}
      ]
      customCurve {defaultValue=3.4028234663852886e+38, postInfinityExtrap=RCCE_Constant, preInfinityExtrap=RCCE_Constant, keys []}
      targetRotation=<rot 0 0 0>
     }
    ]
    rBFParams {distanceMethod=Quaternion, function=Linear, radius=20}
    sourceBones [{boneName=spine_01}, {boneName=spine_02}, {boneName=spine_03}, {boneName=spine_04}, {boneName=spine_05}]
   }
   n1 = Anim Dynamics - Bone: dyn_stetho_disc(ComponentPose=Local To Component(LocalPose=Copy Pose From Mesh() {bUseAttachedParent=true}), Alpha=1, ExternalForce="0,0,0") {
    angularDampingOverride=0.999
    bGravityOverrideInSimSpace=true
    bOverrideAngularDamping=true
    bUseSphericalLimits=true
    alphaBoolBlend {blendOption=Linear}
    boundBone {boneName=dyn_stetho_disc}
    componentLinearAccScale=<1 1 1>
    componentLinearVelScale=<1 1 1>
    physicsBodyDefinitions [{
      sphereCollisionRadius=0
      boundBone {boneName=dyn_stetho_disc}
      boxExtents=<10 10 1>
      constraintSetup {angularLimitsMax=<0 20 0>, angularLimitsMin=<-10 -20 0>, angularTarget=<0 0 0>}
      localJointOffset=<0 0 -1>
     }]
   }
   n2 = Anim Dynamics - Bone: dyn_stetho_earpiece_base(ComponentPose=@n1, Alpha=1, ExternalForce="0,0,0") {
    angularDampingOverride=0.999
    angularSpringConstant=250
    bAngularSpring=true
    bGravityOverrideInSimSpace=true
    bOverrideAngularDamping=true
    gravityScale=0.5
    alphaBoolBlend {blendOption=Linear}
    boundBone {boneName=dyn_stetho_earpiece_base}
    componentLinearAccScale=<0.2 0.2 0.2>
    componentLinearVelScale=<0.2 0.2 0.2>
    physicsBodyDefinitions [{
      sphereCollisionRadius=0
      boundBone {boneName=dyn_stetho_earpiece_base}
      boxExtents=<10 1 10>
      constraintSetup {angularLimitsMax=<0 5 10>, angularLimitsMin=<0 -10 0>, angularTarget=<0 0 0>}
      localJointOffset=<2 0 -0.1>
     }]
   }
   n3 = Anim Dynamics - Bone: dyn_stetho_earpiece_A(ComponentPose=@n2, Alpha=1, ExternalForce="0,0,0") {
    angularDampingOverride=0.999
    bGravityOverrideInSimSpace=true
    bOverrideAngularDamping=true
    gravityScale=0.5
    alphaBoolBlend {blendOption=Linear}
    boundBone {boneName=dyn_stetho_earpiece_A}
    physicsBodyDefinitions [{
      sphereCollisionRadius=0
      boundBone {boneName=dyn_stetho_earpiece_A}
      boxExtents=<1 10 10>
      constraintSetup {angularLimitsMax=<0 2 0>, angularLimitsMin=<0 -10 0>, angularTarget=<0 0 0>}
      localJointOffset=<1 0 0>
     }]
   }
   n4 = "dyn_stetho_earpiece_B.rotateY = dyn_stetho_earpiece_A.rotateY * -0.5 - Bone Driven Controller"!(ComponentPose:Component Space Pose Link Structure=@n3, Alpha:float=1) -> (Pose:Component Space Pose Link Structure) {
    is=title AnimGraphNode_BoneDrivenController
    bAffectTargetRotationY=true
    multiplier=-0.5
    sourceComponent=RotationY
    alphaBoolBlend {blendOption=Linear}
    sourceBone {boneName=dyn_stetho_earpiece_A}
    targetBone {boneName=dyn_stetho_earpiece_B}
   }
   n5 = "dyn_kneepad_r.rotateZ = calf_r.rotateZ * 0.5 - Bone Driven Controller"!(ComponentPose:Component Space Pose Link Structure=@n4, Alpha:float=1) -> (Pose:Component Space Pose Link Structure) {
    is=title AnimGraphNode_BoneDrivenController
    bAffectTargetRotationZ=true
    multiplier=0.5
    sourceComponent=RotationZ
    alphaBoolBlend {blendOption=Linear}
    sourceBone {boneName=calf_r}
    targetBone {boneName=dyn_kneepad_r}
   }
   n6 = Look At - Bone: dyn_front_pouch(ComponentPose=@n5, Alpha=1) {
    interpolationType=Cubic
    alphaBoolBlend {blendOption=Linear}
    boneToModify {boneName=dyn_front_pouch}
    lookAtLocation=<10 10>
    lookAtTarget {boneReference {boneName=spine_04}}
    lookAt_Axis {axis {y=-1}}
   }
   n7 = Get Front End Chest Fw Alpha() -> (Front End Chest Fw Alpha:double)
   n8 = "Transform (Modify) Bone - Bone: dyn_stetho_disc"(ComponentPose=@n6, Alpha=@n7, Translation="0,0,0", Rotation="0,0,-3", Scale="1,1,1") {
    rotationMode=BMM_Additive
    rotationSpace=BCS_BoneSpace
    translationMode=BMM_Additive
    translationSpace=BCS_BoneSpace
    alphaBoolBlend {blendOption=Linear}
    boneToModify {boneName=dyn_stetho_disc}
    rotation {roll=-3}
   }
   n9 = "Transform (Modify) Bone - Bone: dyn_stetho_earpiece_base"(ComponentPose=@n8, Alpha=@n7, Translation="0,0,0", Rotation="0,5,0", Scale="1,1,1") {
    rotationMode=BMM_Additive
    rotationSpace=BCS_BoneSpace
    translationMode=BMM_Additive
    translationSpace=BCS_BoneSpace
    alphaBoolBlend {blendOption=Linear}
    boneToModify {boneName=dyn_stetho_earpiece_base}
    rotation {yaw=5}
   }
   Output Pose(Result=Component To Local(ComponentPose=@n9)) {layerGroup=None}
  ]}
 EventGraph {nodes [
   Event BlueprintUpdateAnimation()
   > n1 = Branch(Condition=Get bIsInFrontEnd())
   n2 = BreakTransform(InTransform=GetDeltaTransformFromRefPose(self=GetOwnerSkeletalMeshComponent(), BoneName=spine_05, BaseName=spine_03))
   > Set Front End Chest Fw Alpha(Front End Chest Fw Alpha:double=MapRangeClamped(Value=BreakRotator(InRot=@n2.Rotation).Yaw, InRangeA=0, InRangeB=5)) -> (Output_Get:double)
   Set Front End Chest Fw Alpha(execute=@n1.else, Front End Chest Fw Alpha=0)
   n3 = float * float(A=Get AimPitch(), B=Get AimOffsetMultiplier() -> (AimOffsetMultiplier:double))
   "Min (Float)"(A:double=@n3, B:double=0) -> (ReturnValue:double) {is=title K2Node_CommutativeAssociativeBinaryOperator}
   Set!(execute, AimOffset_X:double=0, AimOffset_Y:double=@n3, AimOffset_Z:double=0) -> (then, Output_Get:Vector)
  ]}
]
variables [
 AimOffset {type=Vector}
 AimOffsetMultiplier {type=number, default=10}
 Front End Chest Fw Alpha {type=number}
]
